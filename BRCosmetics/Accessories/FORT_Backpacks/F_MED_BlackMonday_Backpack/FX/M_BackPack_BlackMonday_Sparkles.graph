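# Translucent unlit sparkle material for the Black Monday back bling's effects, turning mesh geometry into twinkling flare sprites.
# Vertex colors vary sine timers blended into a blink mask that scales both sprite size and flare opacity.
# Dynamic parameters control size, randomization and expansion along normals through world position offset.
# Emissive pulses the particle color, with a static switch to bypass per-vertex hue shifting.
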Material /BRCosmetics/Accessories/FORT_Backpacks/F_MED_BlackMonday_Backpack/FX/M_BackPack_BlackMonday_Sparkles
graph {
 properties {
  bEnableResponsiveAA=true
  bUsedWithMeshDeformer=true
  bUsedWithMeshParticles=true
  bUsedWithNiagaraSprites=true
  bUsedWithParticleSprites=true
  bUsedWithSkeletalMesh=true
  blendMode=BLEND_Translucent
  opacityMaskClipValue=0.01
  pixelDepthOffsetMode=PDOM_Legacy
  shadingModel=MSM_Unlit
  twoSided=true
  usedShadingModels=Unlit
 }
 nodes [
  n1 = VertexColor()
  n2 = VertexColor()
  n3 = ScalarParameter() {defaultValue=0.35, parameterName=Speed_01}
  n4 = Power(
   Base=MaterialFunctionCall(
    A=MaterialFunctionCall(Speed=@n2.R) {materialFunction=&DebugTimeSine}.Result
    B=MaterialFunctionCall(Speed=@n2.G) {materialFunction=&DebugTimeSine}.Result
    C=MaterialFunctionCall(Speed=@n2.B) {materialFunction=&DebugTimeSine}.Result
    Alpha=MaterialFunctionCall(Speed=@n3) {materialFunction=&DebugTimeSine}.Result
   ) {materialFunction=&3ColorBlend}.Result
   Exp=ScalarParameter() {defaultValue=1, parameterName=BlinkGap}
  ) {constExponent=5}
  n5 = DynamicParameter() {defaultValue=<rgba 100 1 1 1>, paramNames [SpriteSize, Expansion, Sprite Size Scale, Randomizer]}
  n6 = Multiply(
   A=DynamicParameter() {defaultValue=<rgba 100 1 1 1>, paramNames [SpriteSize, Expansion, Sprite Size Scale, Randomizer]}.Randomizer
   B=Multiply(A=@n4, B=@n5.Sprite Size Scale)
  )
  Add(
   A=MaterialFunctionCall!(Center Location=WorldPosition().XYZ, 0-1 UVs, XY scale=Multiply(A=@n1.R, B=@n6), OverridePivot, "Normals (Optional)") {materialFunction=&Sprite}.World position offset
   B=Multiply(A=VertexNormalWS(), B=Multiply(A=@n1.R, B=@n5.Expansion))
  )
  n7 = Time()
  n8 = ScalarParameter() {defaultValue=1, parameterName=SpeedMult}
  n9 = Sine(None=Multiply(A=Multiply(A=@n7, B=@n8) {constB=5}, B=@n2.R))
  n10 = Multiply(A=Clamp(None=Multiply(A=@n9)) {maxDefault=50, minDefault=0.25}, B=ParticleColor().RGB)
  n11 = Multiply(A=VertexColor().R, B=ScalarParameter() {defaultValue=0.1, parameterName=HueShiftMult}) {constB=0.025}
  MaterialFunctionCall(Emissive=StaticSwitchParameter(
    True=@n10
    False=MaterialFunctionCall(Hue Shift Percentage=@n11, Texture=@n10) {materialFunction=&HueShift}.Result
   ) {parameterName=Use Particle Color with NO Hue Shift}) {materialFunction=&MF_FX_EmissiveAdjustments}
  n12 = TextureSampleParameter2D() {parameterName=FlareTexture, samplerType=SAMPLERTYPE_LinearGrayscale, texture=/Game/Environments/FrontEnd/Materials/T_LensFlareSimple}
  n13 = StaticComponentMaskParameter(None=AppendVector(A=@n12.RGB, B=@n12.RGB)) {defaultG=true, parameterName=FlareTexture_Mask}
  Multiply(A=Multiply(A=ParticleColor().A, B=Multiply(A=@n13, B=@n1.R)), B=@n4)
 ]
}
